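# Physics asset for the HazardEdge backpack cosmetic, giving its skeletal mesh a jointed chain that can dangle from the pack.
# One large capsule covers the pack, while small capsules form the chain base and five dynamic links, the last slightly larger.
# The base joins the pack with free swing and twist.
# Each link constrains to the one before it, with angular limits generally widening toward the tip so the end bends most.
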PhysicsAsset /BRCosmetics/Accessories/FORT_Backpacks/Backpack_M_MED_HazardEdge/Meshes/M_MED_HazardEdge_Pack_Physics
bodies {
 base_chain [{shapeType=Capsule, length=0.505, radius=0.505, shapeName=base_chain_capsule, rotation=<rot 0 -90 0>}]
 dyn_chain_01 [{shapeType=Capsule, length=1, radius=1, center=<1 0 0>, rotation=<rot 90 0 0>}]
 dyn_chain_02 [{shapeType=Capsule, length=1, radius=1, center=<1 0 0>, rotation=<rot 90 0 0>}]
 dyn_chain_03 [{shapeType=Capsule, length=1, radius=1, center=<1 0 0>, rotation=<rot 90 0 0>}]
 dyn_chain_04 [{shapeType=Capsule, length=1, radius=1, center=<1 0 0>, rotation=<rot 90 0 0>}]
 dyn_chain_05 [{shapeType=Capsule, length=2, radius=2, center=<2 0 0>, rotation=<rot 90 0 0>}]
 pack [{
   shapeType=Capsule
   length=14.9244
   radius=15.3115
   shapeName=pack_capsule
   center=<-17.8447 19.1314 0.015>
   rotation=<rot -0.0002 -86.958 1.8739>
  }]
}
constraints [
 {
  bone1Name=base_chain
  bone2Name=pack
  swing1LimitDegrees=45
  swing1Motion=Free
  swing2LimitDegrees=45
  swing2Motion=Free
  twistLimitDegrees=45
  twistMotion=Free
 }
 {bone1Name=dyn_chain_01, bone2Name=base_chain, swing1LimitDegrees=15, swing2LimitDegrees=15, twistLimitDegrees=5}
 {bone1Name=dyn_chain_02, bone2Name=dyn_chain_01, swing1LimitDegrees=18, swing2LimitDegrees=18, twistLimitDegrees=8}
 {bone1Name=dyn_chain_03, bone2Name=dyn_chain_02, swing1LimitDegrees=23, swing2LimitDegrees=23, twistLimitDegrees=10}
 {bone1Name=dyn_chain_04, bone2Name=dyn_chain_03, swing1LimitDegrees=25, swing2LimitDegrees=25, twistLimitDegrees=10}
 {bone1Name=dyn_chain_05, bone2Name=dyn_chain_04, swing1LimitDegrees=28, swing2LimitDegrees=28, twistLimitDegrees=10}
]
